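# UI material with alpha-composite blending that draws the backplate for a mode selection entry in matchmaking settings.
# Color and opacity come from a shared overlay backplate function, multiplied by a circle fill, a rounded-corner mask and vertex alpha.
# IsSelected scales the circle's UVs about the center, growing the fill from the middle to reveal the backplate.
# IsDisabled blends two identical backplate inputs, so it has no visible effect.
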Material /MatchmakingSettingsUI/MatchmakingSettings/SettingEntries/ModeSelection/Materials/M_UI_MatchmakingSettings_ModeSelection
graph {
 properties {bAutomaticallySetUsageInEditor=false, blendMode=BLEND_AlphaComposite, materialDomain=MD_UI, usedShadingModels=Default Lit}
 nodes [
  n1 = MaterialFunctionCall() {materialFunction=/MatchmakingSettingsUI/MatchmakingSettings/Materials/MF_UI_MM_OverlayBackplates}
  n2 = ScalarParameter() {parameterName=IsDisabled, sliderMax=1}
  n3 = Add(A=MaterialFunctionCall() {materialFunction=&GetUserInterfaceUV}.Normalized UV, B=Constant2Vector())
  n4 = Constant2Vector() {g=0.5, r=0.5}
  n5 = Multiply(A=NamedRerouteUsage()) {constB=1.51}
  n6 = Add(A=Divide(A=Subtract(A=@n3, B=@n4), B=Max(A=@n5) {constB=0}), B=@n4)
  n7 = Constant2Vector() {g=0.5, r=0.5}
  n8 = Constant() {r=10}
  n9 = Constant() {r=1}
  n10 = Multiply(
   A=MaterialFunctionCall(
    UVs=MaterialFunctionCall(UV=@n6, ScalePivotPoint=@n7) {materialFunction=&MF_UI_UVScaling}.Scaled To Fill
    Size=Constant() {r=1}
   ) {materialFunction=&MF_UI_Circle}.Fill
   B=Multiply(
    A=VertexColor().A
    B=MaterialFunctionCall(Corner Radius=@n8, Radius in Pixels=@n9) {materialFunction=&MF_UI_CoreUI_RoundedMask}.Shape Mask
   )
  )
  Multiply(A=LinearInterpolate(A=@n1.A Opacity, B=@n1.A Opacity, Alpha=@n2), B=@n10)
  Multiply(A=LinearInterpolate(A=@n1.A Color, B=@n1.A Color, Alpha=@n2), B=@n10)
  NamedRerouteDeclaration(None=ScalarParameter() {defaultValue=1, parameterName=IsSelected, sliderMax=1}) {name=IsSelected, nodeColor=<rgba 0 0 0 1>}
 ]
}
