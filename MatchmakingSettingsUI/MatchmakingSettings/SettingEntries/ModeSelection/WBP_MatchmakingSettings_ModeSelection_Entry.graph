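# Modular button widget for a mode selection entry in the matchmaking settings UI, with a centered mode name over a material background and outline block.
# PreConstruct caches the background's dynamic material and resets its selected and disabled states.
# Selection plays a short animation forward or in reverse, driving the material's IsSelected parameter and text color, while hover toggles the animated outline focus.
# List item assignment casts the object into its UIKitButtonViewModel.
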WidgetBlueprint /MatchmakingSettingsUI/MatchmakingSettings/SettingEntries/ModeSelection/WBP_MatchmakingSettings_ModeSelection_Entry
animations [Anim_OnFocus {
  binding_count=2
  length_seconds=0.2
  object=&self:Anim_OnFocus
  sequence {
   bindings [
    B_Setting {
     id=#1
     possessed_class=CommonBorder
     tracks [{
       type=MovieSceneWidgetMaterialTrack
       sections [{type=MovieSceneParameterSection, channels [IsSelected {type=MovieSceneScriptingFloatChannel, keys [[0,0],[12,1]]}]}]
      }]
    }
    Txt_ModeName {
     id=#2
     possessed_class=CommonTextBlock
     tracks [{
       type=MovieSceneColorTrack
       display_name=Color and Opacity
       sections [{
         type=MovieSceneColorSection
         end_frame=12
         start_frame=0
         channels [
          Color.R {type=MovieSceneScriptingFloatChannel, keys [[0,0.5776],[12,0.0048]]}
          Color.G {type=MovieSceneScriptingFloatChannel, keys [[0,0.5776],[12,0.004]]}
          Color.B {type=MovieSceneScriptingFloatChannel, keys [[0,0.5776],[12,0.0116]]}
          Color.A {type=MovieSceneScriptingFloatChannel, const [1,2]}
         ]
        }]
      }]
    }
   ]
   display_rate {numerator=60}
  }
 }]
graphs [
 EventGraph {nodes [
   Event PreConstruct()
   > SetDynamicMaterials()
   > SetInitVisual()
   > Set DisabledState!(execute, self, IsDisabled:bool=false) -> (then) {is=title K2Node_CallFunction}
   Event On Selected()
   n1 = Get Anim_OnFocus() -> (Anim_OnFocus:Widget Animation Object Reference)
   > PlayAnimationForward(InAnimation=@n1)
   n2 = Event OnListItemObjectSet()
   > n3 = Cast To UIKitButtonViewModel!(execute, Object=@n2.ListItemObject) -> (then, CastFailed, AsUI Kit Button View Model:UI Kit Button View Model Object Reference) {is=title K2Node_DynamicCast}
   > Set UIKitButtonViewModel(UIKitButtonViewModel:UI Kit Button View Model Object Reference=@n3) -> (Output_Get:UI Kit Button View Model Object Reference)
   Event On Hovered()
   n4 = Get WBP_UIKit_Block_Outline() -> (WBP_UIKit_Block_Outline:WBP UIKit Block Outline Object Reference)
   > Focused_Visualization!(execute, self:WBP UIKit Block Outline Object Reference=@n4, NeedsToAnimate:bool=true) -> (then)
   Event On Unhovered()
   > Unfocused_Visualization!(execute, self:WBP UIKit Block Outline Object Reference=@n4, NeedsToAnimate:bool=true) -> (then)
   Event On Deselected()
   > PlayAnimationReverse(InAnimation=@n1)
  ]}
 Set DisabledState {nodes [
   n1 = Set DisabledState!() -> (then, IsDisabled:bool=~) {is=title K2Node_FunctionEntry}
   > SetScalarParameterValue(
    self=Get BackgroundDynamicMaterial() -> (BackgroundDynamicMaterial:Material Instance Dynamic Object Reference)
    ParameterName=IsDisabled
    Value=Select(Option 0:float=0, Option 1:float=1, Index=@n1) -> (ReturnValue:float)
   )
   > Set RenderOpacity!(
    execute
    RenderOpacity:float=Select(Option 0=1, Option 1=0.6, Index=@n1)
    self:Widget Object Reference=Get Txt_ModeName() -> (Txt_ModeName:Common Text Object Reference)
   ) -> (then, Output_Get:float)
  ]}
 SetDynamicMaterials {nodes [
   SetDynamicMaterials() {is=title K2Node_FunctionEntry}
   > n1 = GetDynamicMaterial(self:Border Object Reference=Get B_Setting() -> (B_Setting:Common Border Object Reference))
   > Set BackgroundDynamicMaterial(BackgroundDynamicMaterial:Material Instance Dynamic Object Reference=@n1) -> (Output_Get:Material Instance Dynamic Object Reference)
  ]}
 SetInitVisual {nodes [
   SetInitVisual() {is=title K2Node_FunctionEntry}
   > SetScalarParameterValue(self=Get BackgroundDynamicMaterial(), ParameterName=IsSelected, Value=0)
  ]}
]
variables [BackgroundDynamicMaterial {type=MaterialInstanceDynamic}]
widgets {
 info {inheritedWidgetCount=1, parentClass=/Script/FortniteUI.UIKitModularButton, rootWidgetClass=&CommonBorder, widgetCount=5}
 widgets [
  B_Setting {
   widgetClassPath=&CommonBorder
   bIsVariable=true
   widget=B_Setting
   properties {
    background {
     bIsDynamicallyLoaded=false
     drawAs=Image
     imageType=FullColor
     mirroring=NoMirror
     resourceName=None
     resourceObject=/MatchmakingSettingsUI/MatchmakingSettings/SettingEntries/ModeSelection/MaterialInstances/MI_UI_MMSettings_ModeSelect_Bg
     tiling=NoTile
     imageSize=<32 32>
     margin=<ltrb 0 0 0 0>
     outlineSettings {
      bUseBrushTransparency=false
      roundingType=HalfHeightRadius
      width=0
      color {colorUseRule=UseColor_Specified, specifiedColor=<rgba 0 0 0 0>}
      cornerRadii=<0 0 0 0>
     }
     tintColor {colorUseRule=UseColor_Specified, specifiedColor=<rgba 1 1 1 1>}
     uVRegion {bIsValid=false, max=<0 0>, min=<0 0>}
    }
    padding=<ltrb 0 0 0 0>
   }
  }
  O_SettingContainer {
   widgetClassPath=&Overlay
   parent=B_Setting
   widget=O_SettingContainer
   properties {bOverride_Cursor=true}
   slot {class=BorderSlot, properties {padding=<ltrb 0 0 0 0>}}
  }
  Txt_ModeName {
   widgetClassPath=&CommonTextBlock
   bIsVariable=true
   parent=O_SettingContainer
   widget=Txt_ModeName
   properties {
    justification=Center
    text=$Mode Name
    colorAndOpacity {colorUseRule=UseColor_Specified, specifiedColor=<rgba 0.5776 0.5776 0.5776 1>}
    font {
     bForceMonospaced=false
     bMaterialIsStencil=false
     fontMaterial=null
     fontObject=&NotoSans
     letterSpacing=0
     monospacedWidth=1
     size=24
     skewAmount=0
     typefaceFontName=NotoSans-Regular
     outlineSettings {
      bApplyOutlineToDropShadows=false
      bMiteredCorners=false
      bSeparateFillAlpha=false
      outlineMaterial=null
      outlineSize=0
      outlineColor=<rgba 0 0 0 1>
     }
    }
    shadowOffset=<0 0>
   }
   slot {class=OverlaySlot, properties {horizontalAlignment=HAlign_Fill, verticalAlignment=VAlign_Center, padding=<ltrb 32 0 32 0>}}
  }
  WBP_UIKit_Block_Outline {
   widgetClassPath=/FortUILibrary/UIKit/Buttons/Blocks/WBP_UIKit_Block_Outline
   bIsVariable=true
   parent=O_SettingContainer
   widget=WBP_UIKit_Block_Outline
   properties {bottomLeftCornerRadius=20, bottomRightCornerRadius=20, radiusInPixels=true, topLeftCornerRadius=20, topRightCornerRadius=20}
   slot {class=OverlaySlot, properties {horizontalAlignment=HAlign_Fill, verticalAlignment=VAlign_Fill}}
  }
  InputActionWidget {widgetClassPath=&CommonActionWidget, bInherited=true, widget=null}
 ]
}
